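# Physics asset for the face accessory mesh of the female medium Ballerina player character, covering two dynamic hair chains.
# Each side runs from a small base sphere through a parent and four numbered hair capsules, linked in sequence by constraints with narrow swing and very tight twist limits.
# Head, neck, upper spine and both clavicle capsules have no constraints, providing body shapes around the hair chains.
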PhysicsAsset /Game/Characters/Player/Female/Medium/Bodies/F_MED_Ballerina/Meshes/Parts/F_MED_Ballerina_FaceAcc_Physics
bodies {
 clavicle_l [{shapeType=Capsule, length=8.75, radius=6.5, shapeName=clavicle_r_capsule, center=<8.7506 0 -1.629>, rotation=<rot 68.9955 -0.0001 179.9999>}]
 clavicle_r [{shapeType=Capsule, length=8.75, radius=6.5, shapeName=clavicle_r_capsule, center=<-8.7506 0 1.629>, rotation=<rot 68.9955 -0.0001 -0.0001>}]
 dyn_hair_01_l [{
   shapeType=Capsule
   length=3.8946
   radius=4.0179
   shapeName=dyn_hair_01_l_capsule
   center=<4.4049 0 0.7449>
   rotation=<rot 82.5636 -179.9999 -179.9999>
  }]
 dyn_hair_01_r [{shapeType=Capsule, length=2.1738, radius=4.6575, shapeName=dyn_hair_01_r_capsule, center=<4.5598 -0.6103 0>, rotation=<rot 90 180 -176.0611>}]
 dyn_hair_02_l [{shapeType=Capsule, length=4.0395, radius=4.841, shapeName=dyn_hair_02_l_capsule, center=<5.5205 0 0.8973>, rotation=<rot 90 0 0>}]
 dyn_hair_02_r [{shapeType=Capsule, length=4.4172, radius=5.3523, shapeName=dyn_hair_02_r_capsule, center=<7.1864 0 0.7012>, rotation=<rot 90 0 0>}]
 dyn_hair_03_l [{shapeType=Capsule, length=7.3624, radius=5.2031, shapeName=dyn_hair_03_l_capsule, center=<7.1864 0 0>, rotation=<rot 90 0 0>}]
 dyn_hair_03_r [{shapeType=Capsule, length=5.1985, radius=5.3698, shapeName=dyn_hair_03_r_capsule, center=<7.1864 0 0>, rotation=<rot 90 0 0>}]
 dyn_hair_04_l [{
   shapeType=Capsule
   length=11.1378
   radius=5.0739
   shapeName=dyn_hair_04_l_capsule
   center=<6.9524 0 -1.2733>
   rotation=<rot 79.5858 -0.0001 -0.0001>
  }]
 dyn_hair_04_r [{
   shapeType=Capsule
   length=9.3006
   radius=5.1778
   shapeName=dyn_hair_04_r_capsule
   center=<8.8726 0 -1.4881>
   rotation=<rot 78.8944 0.0002 0.0002>
  }]
 dyn_hair_base_l [{shapeType=Sphere, radius=0.5, shapeName=dyn_hair_base_l_sphere}]
 dyn_hair_base_r [{shapeType=Sphere, radius=0.5, shapeName=dyn_hair_base_r_sphere}]
 dyn_hair_parent_l [{
   shapeType=Capsule
   length=1.288
   radius=3.3754
   shapeName=dyn_hair_parent_l_capsule
   center=<2.2214 0 -0.6348>
   rotation=<rot -74.0531 179.9998 -179.9998>
  }]
 dyn_hair_parent_r [{shapeType=Capsule, length=1.9124, radius=3.4942, shapeName=dyn_hair_parent_r_capsule, center=<2.8837 0 0>, rotation=<rot 90 0 0>}]
 head [{shapeType=Capsule, length=10, radius=7.3218, shapeName=head_capsule, center=<5.146 0.8212 0>}]
 neck_02 [{shapeType=Capsule, length=3.9667, radius=5.6003, shapeName=neck_02_capsule, center=<3.0143 0 0>, rotation=<rot 90 0 0>}]
 spine_05 [{shapeType=Capsule, length=6.9763, radius=12.2152, shapeName=spine_05_capsule, center=<0 2.8942 0>, rotation=<rot 90 0 0>}]
}
constraints [
 {bone1Name=dyn_hair_04_r, bone2Name=dyn_hair_03_r, swing1LimitDegrees=10, swing2LimitDegrees=20, twistLimitDegrees=1}
 {bone1Name=dyn_hair_parent_l, bone2Name=dyn_hair_base_l, swing1LimitDegrees=10, swing2LimitDegrees=20, twistLimitDegrees=1}
 {bone1Name=dyn_hair_01_l, bone2Name=dyn_hair_parent_l, swing1LimitDegrees=10, swing2LimitDegrees=20, twistLimitDegrees=1}
 {bone1Name=dyn_hair_02_l, bone2Name=dyn_hair_01_l, swing1LimitDegrees=10, swing2LimitDegrees=20, twistLimitDegrees=1}
 {bone1Name=dyn_hair_03_l, bone2Name=dyn_hair_02_l, swing1LimitDegrees=10, swing2LimitDegrees=20, twistLimitDegrees=1}
 {bone1Name=dyn_hair_04_l, bone2Name=dyn_hair_03_l, swing1LimitDegrees=10, swing2LimitDegrees=20, twistLimitDegrees=1}
 {bone1Name=dyn_hair_parent_r, bone2Name=dyn_hair_base_r, swing1LimitDegrees=10, swing2LimitDegrees=20, twistLimitDegrees=1}
 {bone1Name=dyn_hair_01_r, bone2Name=dyn_hair_parent_r, swing1LimitDegrees=10, swing2LimitDegrees=20, twistLimitDegrees=1}
 {bone1Name=dyn_hair_02_r, bone2Name=dyn_hair_01_r, swing1LimitDegrees=10, swing2LimitDegrees=20, twistLimitDegrees=1}
 {bone1Name=dyn_hair_03_r, bone2Name=dyn_hair_02_r, swing1LimitDegrees=10, swing2LimitDegrees=20, twistLimitDegrees=1}
]
