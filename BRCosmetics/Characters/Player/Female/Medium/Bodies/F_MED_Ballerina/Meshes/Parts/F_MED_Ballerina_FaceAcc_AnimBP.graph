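# Post-copy animation Blueprint for the Ballerina character's face accessory part, adding physics motion to its hair.
# It copies the attached parent's pose, additively rotates both hair base bones, then runs RigidBody simulation in head-bone space.
# Each update computes a gravity override through a deprecated state function, with tuned skydive parameters.
# The hair rotation applies only while skydiving with rigid body flagged disabled, by setting or LOD above 1.
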AnimBlueprint /BRCosmetics/Characters/Player/Female/Medium/Bodies/F_MED_Ballerina/Meshes/Parts/F_MED_Ballerina_FaceAcc_AnimBP
target_skeleton=/Game/Characters/Player/Female/Medium/Bodies/F_MED_Ballerina/Meshes/Parts/F_MED_Ballerina_FaceAcc_Skeleton
graphs [
 AnimGraph {nodes [
   n1 = Get HairAdjustment() -> (HairAdjustment:double)
   n2 = "Transform (Modify) Bone - Bone: dyn_hair_base_l"(
    ComponentPose=Local To Component(LocalPose=Copy Pose From Mesh() {bUseAttachedParent=true})
    Alpha=@n1
    Translation="0,0,0"
    Rotation="-30,0,10"
    Scale="1,1,1"
   ) {
    rotationMode=BMM_Additive
    rotationSpace=BCS_BoneSpace
    translationMode=BMM_Additive
    translationSpace=BCS_BoneSpace
    alphaBoolBlend {blendOption=Linear}
    boneToModify {boneName=dyn_hair_base_l}
    rotation {pitch=-30, roll=10}
   }
   n3 = "Transform (Modify) Bone - Bone: dyn_hair_base_r"(ComponentPose=@n2, Alpha=@n1, Translation="0,0,0", Rotation="-30,0,-10", Scale="1,1,1") {
    rotationMode=BMM_Additive
    rotationSpace=BCS_BoneSpace
    translationMode=BMM_Additive
    translationSpace=BCS_BoneSpace
    alphaBoolBlend {blendOption=Linear}
    boneToModify {boneName=dyn_hair_base_r}
    rotation {pitch=-30, roll=-10}
   }
   n4 = RigidBody!(ComponentPose=@n3, Alpha, OverrideWorldGravity:Vector=Get Gravity_Override() -> (Gravity_Override:Vector), ExternalForce) -> (Pose) {
    bClampLinearTranslationLimitToRefPose=true
    bOverrideWorldGravity=true
    overridePhysicsAsset=/Game/Characters/Player/Female/Medium/Bodies/F_MED_Ballerina/Meshes/Parts/F_MED_Ballerina_FaceAcc_Physics
    simulationSpace=BaseBoneSpace
    alphaBoolBlend {blendOption=Linear}
    baseBoneRef {boneName=head}
    componentAppliedLinearAccClamp=<3000 3000 3000>
    componentLinearAccScale=<0.3 0.2 0.2>
    componentLinearVelScale=<0.3 0.2 0.2>
    overrideWorldGravity {z=-980}
    simSpaceSettings {maxAngularAcceleration=50, maxAngularVelocity=4, maxLinearAcceleration=300, maxLinearVelocity=200, velocityScaleZ=0.3, worldAlpha=1}
   }
   Output Pose(Result=Component To Local(ComponentPose=@n4)) {layerGroup=None}
  ]}
 EventGraph {nodes [
   Event BlueprintUpdateAnimation()
   > n1 = Sequence!(execute) -> (then_0, then_1, then_2:Exec)
   > n2 = Gravity Override State Function DEPRECATED(CharacterPartAnimInstance=Self-Reference(), GameTimeInSeconds=GetGameTimeInSeconds(), GravityOverrideParams=Get Gravity Override Params())
   > Set Gravity_Override(Gravity_Override:Vector=@n2) -> (Output_Get:Vector)
   n3 = Get CurrentLOD()
   n4 = Get bRigidBodyEnabled()
   Set isRigidBodyDisabled(
    execute=@n1.then_1
    isRigidBodyDisabled:bool=OR Boolean(A=NOT Boolean(A=@n4), B="integer > integer"(A=@n3, B=1) {is=title K2Node_CallFunction})
   ) -> (Output_Get:bool)
   n5 = Get isRigidBodyDisabled() -> (isRigidBodyDisabled:bool)
   n6 = Get bIsSkyDiving()
   Set HairAdjustment(execute=@n1.then_2, HairAdjustment:double=SelectFloat(A=1, B=0, bPickA=AND Boolean(A=@n5, B=@n6))) -> (Output_Get:double)
  ]}
]
variables [
 Current LOD {type=integer}
 Gravity Override Params {
  type=GravityOverrideParamsStruct
  default {skydive_FloatingDown {gravityOverride=<0 -500 400>, windAmplitude=<50 -300 200>, windFrequency=<1 4 5>, windOffset=<0 -400 200>}}
 }
 Gravity_Override {type=Vector}
 HairAdjustment {type=number}
 isRigidBodyDisabled {type=boolean}
]
